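# Unlit, color-tintable material for the Elpulpo Infinity skeletal wheel cosmetic, with glowing infinity-depth patterns on rim, tire and center.
# A shared offset function fakes depth from RGB mask textures, with panning streak noise distorting the tire and center layers.
# Masks, a time-driven sine pulse on the green channel and normal-mapped Fresnel build the emissive.
# Vertex color channels supply masks for blending the regions, and the center has its own color and intensity.
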Material /VehicleCosmetics/Wheels/Elpulpo_Infinity/Materials/M_Wheel_Elpulpo_Infinity
graph {
 properties {
  bAutomaticallySetUsageInEditor=false
  bUseMaterialAttributes=true
  bUsedWithMeshDeformer=true
  bUsedWithSkeletalMesh=true
  shadingModel=MSM_Unlit
  usedShadingModels=Unlit
 }
 nodes [
  n1 = NamedRerouteUsage()
  n2 = NamedRerouteUsage()
  Add(A=@n1, B=@n2)
  n3 = VectorParameter() {group=00 Base, parameterName=Custom Color, sortPriority=-1, defaultValue=<rgba 0.0497 0.1144 0.0307 1>}
  n4 = TextureObjectParameter() {group=00 Base, parameterName=Rim Masks, sortPriority=0, texture=/VehicleCosmetics/Wheels/Elpulpo_Infinity/Textures/elpulpo_infinity_RIM_RGB}
  n5 = ScalarParameter() {defaultValue=0.003, group=00 Base, parameterName=Gradient Darkness, sortPriority=6}
  n6 = MaterialFunctionCall!(
   Texture=@n4
   Offset=ScalarParameter() {defaultValue=0.005, group=00 Base, parameterName=Bump Offset Amount, sortPriority=5}
   UVs
   RotateUVs
   GradientFalloff
   GradientDarkness=@n5
  ) {materialFunction=/VehicleCosmetics/Wheels/ClockGear_SE/Materials/MF_Infinity_OffsetFunction}
  n7 = TextureObjectParameter() {
   group=00 Base
   parameterName=Tire Masks
   sortPriority=1
   texture=/VehicleCosmetics/Wheels/Elpulpo_Infinity/Textures/elpulpo_infinity_Tread_RGB
  }
  n8 = TextureSample(UVs=Panner(Coordinate=TextureCoordinate() {uTiling=0.2}) {speedX=0.1}) {samplerSource=SSM_Wrap_WorldGroupSettings, texture=/VehicleCosmetics/Wheels/_SharedTextures/FX/_3D_ENV_VerticalStreaks_skin}
  n9 = MaterialFunctionCall(Texture=@n7, Offset=Multiply(A=ConstantBiasScale(None=@n8.RGB) {bias=-0.1, scale=2}) {constB=-1}, GradientDarkness=@n5) {materialFunction=&MF_Infinity_OffsetFunction}
  n10 = NamedRerouteUsage()
  n11 = TextureSample(Tex=@n4) {samplerSource=SSM_Wrap_WorldGroupSettings, texture=&_3D_ENV_VerticalStreaks_skin}
  n12 = TextureSample(Tex=@n7) {samplerSource=SSM_Wrap_WorldGroupSettings, texture=&_3D_ENV_VerticalStreaks_skin}
  n13 = LinearInterpolate(A=@n11.B, B=@n12.B, Alpha=@n10) {desc=B}
  n14 = LinearInterpolate(B=ComponentMask(None=LinearInterpolate(A=@n6.Result, B=@n9.Result, Alpha=@n10)) {r=true}, Alpha=@n13)
  n15 = LinearInterpolate(A=@n11.R, B=@n12.R, Alpha=@n10) {desc=R}
  n16 = ScalarParameter() {defaultValue=2, group=00 Base, parameterName=Emissive Intensity, sortPriority=7}
  n17 = Time()
  n18 = Abs(None=Sine(None=Multiply(A=@n17) {constB=0.125}))
  n19 = Max(
   A=Multiply(A=Max(A=@n14, B=@n15), B=Multiply(A=@n16, B=@n16))
   B=Multiply(A=LinearInterpolate(A=@n11.G, B=@n12.G, Alpha=@n10) {desc=G}, B=ConstantBiasScale(None=@n18) {bias=0.7, scale=2})
  )
  n20 = TextureSampleParameter2D() {
   group=00 Base
   parameterName=Rim Normal
   samplerType=SAMPLERTYPE_Normal
   sortPriority=2
   texture=/VehicleCosmetics/Wheels/Elpulpo_Infinity/Textures/Wheel_Elpulpo_N
  }
  n21 = TextureSampleParameter2D() {
   group=00 Base
   parameterName=Tire Normal
   samplerType=SAMPLERTYPE_Normal
   sortPriority=3
   texture=/VehicleCosmetics/Wheels/Elpulpo_Infinity/Textures/Tire_Elpulpo_N
  }
  n22 = LinearInterpolate(A=@n20.RGB, B=LinearInterpolate(A=@n21.RGB, B=Constant3Vector() {constant=<rgba 0 0 1 0>}, Alpha=@n12.B), Alpha=NamedRerouteUsage())
  n23 = Constant() {r=3}
  n24 = Max(
   A=Multiply(A=Clamp(None=@n19) {maxDefault=10}, B=@n16)
   B=Clamp(None=MaterialFunctionCall!(Normal=@n22, Exponent=@n23) {materialFunction=/VehicleCosmetics/Wheels/Cola_Holographic/Materials/MF_Fresnel_Classic}.Result) {minDefault=0.1}
  )
  n25 = NamedRerouteUsage()
  NamedRerouteDeclaration(None=Multiply(A=Multiply(A=@n3.RGB, B=@n24), B=OneMinus(None=@n25))) {name=Emissive, nodeColor=<rgba 0 0.6484 1 1>}
  n26 = VertexColor()
  NamedRerouteDeclaration(None=@n26.R) {name=VC_R, nodeColor=<rgba 1 0 0.0234 1>}
  n27 = MaterialFunctionCall(Normal=@n22, Exponent=Constant() {r=1}) {materialFunction=&MF_Fresnel_Classic}
  NamedRerouteDeclaration(None=LinearInterpolate(A=Multiply(A=@n27.Result, B=@n3.RGB), Alpha=@n13) {constB=0}) {name=BC, nodeColor=<rgba 1 0 0.0234 1>}
  NamedRerouteDeclaration(None=@n26.G) {name=VC_G, nodeColor=<rgba 0 1 0 1>}
  n28 = VectorParameter() {group=01 Center, parameterName=Center Custom Color, sortPriority=-1, defaultValue=<rgba 1 1 1 1>}
  n29 = TextureObjectParameter() {group=01 Center, parameterName=Center, sortPriority=0, texture=/VehicleCosmetics/Wheels/Elpulpo_Infinity/Textures/Wheel_Elpulpo_Inf}
  n30 = Panner(Coordinate=TextureCoordinate() {uTiling=0.05}) {speedX=0.1}
  n31 = Multiply(
   A=ConstantBiasScale(None=TextureSample(UVs=@n30) {samplerSource=SSM_Wrap_WorldGroupSettings, texture=&_3D_ENV_VerticalStreaks_skin}.RGB) {bias=0.15, scale=2}
   B=ScalarParameter() {defaultValue=-10, group=01 Center, parameterName=Center Bump Offset Amount, sortPriority=0}
  ) {constB=-10}
  n32 = ScalarParameter() {defaultValue=0.001, group=01 Center, parameterName=Center Gradient Darkness, sortPriority=1}
  n33 = TextureSample(Tex=@n29) {samplerSource=SSM_Wrap_WorldGroupSettings, texture=&_3D_ENV_VerticalStreaks_skin}
  n34 = LinearInterpolate(
   A=ComponentMask(None=MaterialFunctionCall(Texture=@n29, Offset=@n31, GradientDarkness=@n32) {materialFunction=&MF_Infinity_OffsetFunction}.Result) {r=true}
   Alpha=@n33.B
  ) {constB=0}
  n35 = Multiply(
   A=Clamp(None=Max(A=@n34, B=@n33.R))
   B=ScalarParameter() {defaultValue=1.5, group=01 Center, parameterName=Center Emissive Intensity, sortPriority=2}
  )
  NamedRerouteDeclaration(None=Multiply(A=Multiply(A=@n28.RGB, B=@n35), B=NamedRerouteUsage())) {name=Center Emissive, nodeColor=<rgba 1 0.375 0 1>}
  SetMaterialAttributes!(MaterialAttributes, Base Color=NamedRerouteUsage(), Emissive Color=LinearInterpolate(A=@n2, B=@n1, Alpha=NamedRerouteUsage())) {attributeSetTypes [#1, #2]}
 ]
}
